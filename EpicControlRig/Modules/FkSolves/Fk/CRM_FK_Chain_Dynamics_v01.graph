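# Modular Control Rig module building an FK chain with dynamics between Start and End connectors.
# Construction spawns main controls every Distribution Every bones, nested secondary controls per bone, orient spaces, and shared dynamics and visibility channels.
# Forward solve spreads main control rotation to secondary controls via rotation offsets, drives bones from those, then runs Local Chain Dynamics.
# Backward solve sets one control per bone from the bone transforms.
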ControlRigBlueprint /EpicControlRig/Modules/FkSolves/Fk/CRM_FK_Chain_Dynamics_v01
graphs [
 ConstructionGraph {nodes [
   RigUnit_PrepareForExecution = &RigUnit_PrepareForExecution()
   RigVMFunction_Sequence = RigVMAggregateNode(ExecuteContext=@RigUnit_PrepareForExecution.ExecutePin)
   RigVMVariableNode(ExecuteContext=@RigVMFunction_Sequence.B) {variable=Previous Int}
   > RigVMDispatchNode(Array=RigVMVariableNode() {variable=Main Bones}.Value) {template="DISPATCH_RigVMDispatch_ArrayReset(io Array)"}
   Divide = /Script/RigVM.RigVMFunction_MathIntDiv(
    A=RigVMDispatchNode(Array=RigVMVariableNode() {variable=Bones}.Value) {template="DISPATCH_RigVMDispatch_ArrayGetNum(in Array,out Num)"}.Num
    B=RigVMVariableNode() {variable=Distribution Every}.Value
   ) {is=struct RigVMUnitNode, template="Divide::Execute(in A,in B,out Result)"}
   > RigVMFunction_ForLoopCount = /Script/RigVM.RigVMFunction_ForLoopCount(Count=@Divide.Result) {is=struct RigVMUnitNode}
   > RigVMDispatchNode(
    Element=RigVMDispatchNode(Array=RigVMVariableNode() {variable=Bones}.Value, Index=RigVMVariableNode() {variable=Previous Int}.Value) {template="DISPATCH_RigVMDispatch_ArrayGetAtIndex(in Array,in Index,out Element)"}.Element
    Array=RigVMVariableNode() {variable=Main Bones}.Value
   ) {template="DISPATCH_RigVMDispatch_ArrayAdd(io Array,in Element,out Index)"}
   > RigVMVariableNode(Value=/Script/RigVM.RigVMFunction_MathIntAdd(B=RigVMVariableNode() {variable=Distribution Every}.Value, A=RigVMVariableNode() {variable=Previous Int}.Value) {is=struct RigVMUnitNode, template="Add::Execute(in A,in B,out Result)"}.Result) {variable=Previous Int}
   Construct Space Nulls = RigVMFunctionReferenceNode(
    Orient Only=true
    Driven Control=RigVMDispatchNode(Index=0, Array=RigVMVariableNode() {variable=Controls}.Value) {template="DISPATCH_RigVMDispatch_ArrayGetAtIndex(in Array,in Index,out Element)"}.Element
    Spaces=RigVMVariableNode() {variable=Orient Spaces}.Value
    Space Null Parent=RigVMVariableNode() {variable=Parent}.Value
    ExecuteContext=@RigVMFunction_Sequence.D
   )
   > n1 = RigVMVariableNode(Value=@"Construct Space Nulls".Output Space Nulls) {variable=Orient Space Nulls}
   Num = RigVMDispatchNode(Array=RigVMVariableNode() {variable=Orient Spaces}.Value) {template="DISPATCH_RigVMDispatch_ArrayGetNum(in Array,out Num)"}
   n2 = RigVMVariableNode() {variable=Default Orient Space Index}
   Not_Equals = RigVMDispatchNode(B=-1, A=@n2.Value) {template="DISPATCH_RigVMDispatch_CoreNotEquals(in A,in B,out Result)"}
   RigVMFunction_ControlFlowBranch = &RigVMFunction_ControlFlowBranch(
    Condition=&RigVMFunction_MathBoolAnd(
     B=/Script/RigVM.RigVMFunction_MathIntLess(B=@Num.Num, A=@n2.Value) {is=struct RigVMUnitNode, template="Less::Execute(in A,in B,out Result)"}.Result
     A=@Not_Equals.Result
    ).Result
    ExecuteContext=@n1.ExecuteContext
   )
   /Script/ControlRig.RigUnit_SwitchParent(
    Mode=ParentItem
    bMaintainGlobal=True
    Parent=RigVMDispatchNode(Array=RigVMVariableNode() {variable=Orient Space Nulls}.Value, Index=@n2.Value) {template="DISPATCH_RigVMDispatch_ArrayGetAtIndex(in Array,in Index,out Element)"}.Element
    Child=RigVMDispatchNode(Index=0, Array=RigVMVariableNode() {variable=Controls}.Value) {template="DISPATCH_RigVMDispatch_ArrayGetAtIndex(in Array,in Index,out Element)"}.Element
    ExecutePin=@RigVMFunction_ControlFlowBranch.True
   ) {is=struct RigVMUnitNode}
   n3 = RigVMDispatchNode(Index=0, Array=RigVMVariableNode() {variable=Controls}.Value) {template="DISPATCH_RigVMDispatch_ArrayGetAtIndex(in Array,in Index,out Element)"}
   SpawnAnimationChannel = /Script/ControlRig.RigUnit_HierarchyAddAnimationChannelFloat(
    InitialValue=3
    LimitsEnabled=(Enabled=(bMinimum=True,bMaximum=True))
    MaximumValue=40
    MinimumValue=0
    Name=Dyn Strength
    Parent=@n3.Element
    ExecutePin=@RigVMFunction_Sequence.E
   ) {
    is=struct RigVMUnitNode
    template="SpawnAnimationChannel::Execute(in InitialValue,in MinimumValue,in MaximumValue,in LimitsEnabled,in Parent,in Name,out Item)"
   }
   > RigVMVariableNode(Value=@SpawnAnimationChannel.Item) {variable=Dyn Strength Channel}
   > n4 = &RigUnit_HierarchyAddAnimationChannelFloat(
    InitialValue=0.7
    LimitsEnabled=(Enabled=(bMinimum=True,bMaximum=True))
    MaximumValue=40
    MinimumValue=0
    Name=Dyn Damping
    Parent=@n3.Element
   ) {template="SpawnAnimationChannel::Execute(in InitialValue,in MinimumValue,in MaximumValue,in LimitsEnabled,in Parent,in Name,out Item)"}
   > RigVMVariableNode(Value=@n4.Item) {variable=Dyn Damping Channel}
   > n5 = &RigUnit_HierarchyAddAnimationChannelFloat(
    InitialValue=0.2
    LimitsEnabled=(Enabled=(bMinimum=True,bMaximum=True))
    MaximumValue=1
    MinimumValue=0
    Name=Dyn Blend Start
    Parent=@n3.Element
   ) {template="SpawnAnimationChannel::Execute(in InitialValue,in MinimumValue,in MaximumValue,in LimitsEnabled,in Parent,in Name,out Item)"}
   > RigVMVariableNode(Value=@n5.Item) {variable=Dyn Blend Start Channel}
   > n6 = &RigUnit_HierarchyAddAnimationChannelFloat(
    InitialValue=1
    LimitsEnabled=(Enabled=(bMinimum=True,bMaximum=True))
    MaximumValue=1
    MinimumValue=0
    Name=Dyn Blend End
    Parent=@n3.Element
   ) {template="SpawnAnimationChannel::Execute(in InitialValue,in MinimumValue,in MaximumValue,in LimitsEnabled,in Parent,in Name,out Item)"}
   > RigVMVariableNode(Value=@n6.Item) {variable=Dyn Blend End Channel}
   > n7 = /Script/ControlRig.RigUnit_HierarchyAddAnimationChannelBool(InitialValue=False, LimitsEnabled=(), MaximumValue=True, MinimumValue=False, Name=Sec Control Vis, Parent=@n3.Element) {
    is=struct RigVMUnitNode
    template="SpawnAnimationChannel::Execute(in InitialValue,in MinimumValue,in MaximumValue,in LimitsEnabled,in Parent,in Name,out Item)"
   }
   > RigVMVariableNode(Value=@n7.Item) {variable=Sec Control Vis Channel}
   Clone = RigVMDispatchNode(Array=RigVMVariableNode() {variable=Controls}.Value) {template="DISPATCH_RigVMDispatch_ArrayClone(in Array,out Clone)"}
   > Remove = RigVMDispatchNode(Index=0, Array=@Clone.Clone) {template="DISPATCH_RigVMDispatch_ArrayRemove(io Array,in Index)"}
   Item Array = RigVMDispatchNode(
    Value=((Type=None,Name="None"),(Type=None,Name="None"),(Type=None,Name="None"),(Type=None,Name="None"),(Type=None,Name="None"))
    Value.3=RigVMVariableNode() {variable=Dyn Strength Channel}.Value
    Value.2=RigVMVariableNode() {variable=Dyn Damping Channel}.Value
    Value.1=RigVMVariableNode() {variable=Dyn Blend Start Channel}.Value
    Value.0=RigVMVariableNode() {variable=Dyn Blend End Channel}.Value
    Value.4=RigVMVariableNode() {variable=Sec Control Vis Channel}.Value
   ) {template="DISPATCH_RigVMDispatch_Constant(io Value)"}
   > For_Each_1 = RigVMDispatchNode(Array=@"Item Array".Value) {template="DISPATCH_RigVMDispatch_ArrayIterator(in Array,out Element,out Index,out Count,out Ratio)"}
   > /Script/ControlRig.RigUnit_SetChannelHosts(Channel=@For_Each_1.Element, Hosts=@Remove.Array) {is=struct RigVMUnitNode}
   DISPATCH_RigVMDispatch_ArrayReset = RigVMDispatchNode(Array=RigVMVariableNode() {variable=Controls}.Value, ExecuteContext=@RigVMFunction_Sequence.A) {template="DISPATCH_RigVMDispatch_ArrayReset(io Array)"}
   n8 = &RigVMFunction_ControlFlowBranch(
    ExecuteContext=@DISPATCH_RigVMDispatch_ArrayReset.ExecuteContext
    Condition=/Script/ControlRig.RigUnit_ResolveConnector(Connector=(Type=Connector,Name="End"), SkipSocket=False) {is=struct RigVMUnitNode}.bIsConnected
   )
   RigVMVariableNode(Value=(Type=Connector,Name="Parent"), ExecuteContext=@n8.Completed) {variable=Parent}
   > n9 = RigVMVariableNode(Value=(Type=Connector,Name="Parent")) {variable=Fk Parent}
   n10 = /Script/ControlRig.RigUnit_ResolveArrayConnector(Connector=(Type=Connector,Name="Orient Spaces"), SkipSocket=False) {is=struct RigVMUnitNode}
   n11 = &RigVMFunction_ControlFlowBranch(Condition=@n10.bIsConnected, ExecuteContext=@n9.ExecuteContext)
   RigVMVariableNode(Value=@n10.Result, ExecuteContext=@n11.True) {variable=Orient Spaces}
   n12 = /Script/ControlRig.RigUnit_HierarchyGetChainItemArray(End=(Type=Connector,Name="End"), Start=(Type=Connector,Name="Start"), bIncludeEnd=True, bIncludeStart=True, bReverse=False) {is=struct RigVMUnitNode}
   RigVMVariableNode(Value=@n12.Chain, ExecuteContext=@n8.True) {variable=Bones}
   RigVMVariableNode(
    Value=((Type=None,Name="None"))
    Value.0=&RigUnit_ResolveConnector(Connector=(Type=Connector,Name="Start"), SkipSocket=False).Result
    ExecuteContext=@n8.False
   ) {variable=Bones}
   n13 = RigVMDispatchNode(
    Array=RigVMDispatchNode(Array=RigVMVariableNode() {variable=Controls}.Value) {template="DISPATCH_RigVMDispatch_ArrayClone(in Array,out Clone)"}.Clone
    Other=RigVMVariableNode() {variable=Sec Controls}.Value
    ExecuteContext=@RigVMFunction_Sequence.F
   ) {template="DISPATCH_RigVMDispatch_ArrayAppend(io Array,in Other)"}
   Get_Module_Metadata_1 = RigVMDispatchNode(Default=1, Name=Global Control Scale, NameSpace=Root) {template="DISPATCH_RigDispatch_GetModuleMetadata(in Name,in NameSpace,in Default,out Value,out Found)"}
   > Scale Module Controls = RigVMFunctionReferenceNode(Controls=@n13.Array, Global Scale=@Get_Module_Metadata_1.Value, Local Scale=RigVMVariableNode() {variable=Control Scale}.Value)
   RigVMDispatchNode(Array=RigVMVariableNode() {variable=Backward Controls}.Value, ExecuteContext=@RigVMFunction_Sequence.C) {template="DISPATCH_RigVMDispatch_ArrayReset(io Array)"}
   > RigVMDispatchNode(Array=RigVMVariableNode() {variable=Sec Controls}.Value) {template="DISPATCH_RigVMDispatch_ArrayReset(io Array)"}
   > DISPATCH_RigVMDispatch_ArrayIterator = RigVMDispatchNode(Array=RigVMVariableNode() {variable=Bones}.Value) {template="DISPATCH_RigVMDispatch_ArrayIterator(in Array,out Element,out Index,out Count,out Ratio)"}
   > n14 = RigVMDispatchNode(Array=RigVMVariableNode() {variable=Nulls}.Value) {template="DISPATCH_RigVMDispatch_ArrayReset(io Array)"}
   n15 = &RigVMFunction_ControlFlowBranch(
    Condition=RigVMDispatchNode(Element=@DISPATCH_RigVMDispatch_ArrayIterator.Element, Array=RigVMVariableNode() {variable=Main Bones}.Value) {template="DISPATCH_RigVMDispatch_ArrayFind(in Array,in Element,out Index,out Success)"}.Success
    ExecuteContext=@n14.ExecuteContext
   )
   Int to Name = /Script/RigVM.RigVMFunction_MathIntToName(PaddedSize=0, Number=@DISPATCH_RigVMDispatch_ArrayIterator.Index) {is=struct RigVMUnitNode, template="Int to Name::Execute(in Number,in PaddedSize,out Result)"}
   n16 = /Script/RigVM.RigVMFunction_MathTransformMakeAbsolute(
    Parent=&RigUnit_GetTransform(bInitial=true, Item=@DISPATCH_RigVMDispatch_ArrayIterator.Element).Transform
    Local=RigVMVariableNode() {variable=Control Transform Offset}.Value
   ) {is=struct RigVMUnitNode, template="Make Absolute::Execute(in Local,in Parent,out Global)"}
   n17 = RigVMVariableNode() {variable=Sec Shape Settings}
   n18 = &RigVMFunction_MathTransformMul(
    B=/Script/RigVM.RigVMFunction_MathTransformInverse(Value=RigVMVariableNode() {variable=Control Transform Offset}.Value) {is=struct RigVMUnitNode, template="Inverse::Execute(in Value,out Result)"}.Result
    A=@n17.Value.Transform
   ) {template="Multiply::Execute(in A,in B,out Result)"}
   RigVMFunction_AnimEvalRichCurve = &RigVMFunction_AnimEvalRichCurve(
    SourceMinimum=0
    TargetMinimum=0
    Curve=RigVMVariableNode() {variable=Control Scale Factor Profile}.Value
    Value=@DISPATCH_RigVMDispatch_ArrayIterator.Ratio
   )
   n19 = /Script/RigVM.RigVMFunction_MathVectorScale(Value=(X=0,Y=0,Z=0), Value.X=@n18.Result.Scale3D.X, Value.Y=@n18.Result.Scale3D.Y, Factor=@RigVMFunction_AnimEvalRichCurve.Result) {is=struct RigVMUnitNode, template="Scale::Execute(in Value,in Factor,out Result)"}
   n20 = RigVMDispatchNode(
    Elements=(bVisible=True,Name="Circle_Thick",Color=(R=1,G=0,B=0,A=1),Transform=(Rotation=(X=0,Y=-0.7071,Z=0,W=0.7071),Translation=(X=0,Y=0,Z=0),Scale3D=(X=0.4,Y=0.4,Z=0.4)))
    Elements.Name=RigVMDispatchNode(
     False=Sphere_Solid
     Condition=/Script/ControlRig.RigUnit_ShapeExists(ShapeName=@n17.Value.Name) {is=struct RigVMUnitNode}.Result
     True=@n17.Value.Name
    ) {template="DISPATCH_RigVMDispatch_If(in Condition,in True,in False,out Result)"}.Result
    Elements.Transform.Rotation=@n18.Result.Rotation
    Elements.Transform.Scale3D.Z=@n18.Result.Scale3D.Z
    Elements.Transform.Translation=@n18.Result.Translation
    Elements.Transform.Scale3D.X=@n19.Result.X
    Elements.Transform.Scale3D.Y=@n19.Result.Y
    Elements.Color=@n17.Value.Color
   ) {template="DISPATCH_RigVMDispatch_MakeStruct(in Elements,out Struct)"}
   n21 = /Script/ControlRig.RigUnit_HierarchyAddControlTransform(
    InitialValue=(Rotation=(X=0,Y=0,Z=0,W=1),Translation=(X=0,Y=0,Z=0),Scale3D=(X=1,Y=1,Z=1))
    OffsetSpace=GlobalSpace
    Settings="(InitialSpace=LocalSpace,bUsePreferredRotationOrder=False,PreferredRotationOrder=YZX,Limits=(LimitTranslationX=(bMinimum=False,bMaximum=False),LimitTranslationY=(bMinimum=False,bMaximum=False),LimitTranslationZ=(bMinimum=False,bMaximum=False),LimitPitch=(b ...(truncated)"
    Name=/Script/RigVM.RigVMFunction_NameConcat(A="Sec FK ", B=@"Int to Name".Result) {is=struct RigVMUnitNode, template="Concat::Execute(in A,in B,out Result)"}.Result
    OffsetTransform=@n16.Global
    Settings.Shape=@n20.Struct
    ExecutePin=@n15.False
    Parent=RigVMVariableNode() {variable=Fk Parent}.Value
   ) {
    is=struct RigVMUnitNode
    template="SpawnControl::Execute(in InitialValue,in Settings,in OffsetTransform,in OffsetSpace,in Parent,in Name,out Item)"
   }
   > RigVMVariableNode(Value=@n21.Item) {variable=Fk Parent}
   > RigVMDispatchNode(Element=@n21.Item, Array=RigVMVariableNode() {variable=Sec Controls}.Value) {template="DISPATCH_RigVMDispatch_ArrayAdd(io Array,in Element,out Index)"}
   > Set_Metadata_1 = RigVMDispatchNode(
    Name=Offset
    NameSpace=Self
    Value=&RigUnit_GetTransform(Space=LocalSpace, bInitial=true, Item=@DISPATCH_RigVMDispatch_ArrayIterator.Element).Transform.Rotation
    Item=@n21.Item
   ) {template="DISPATCH_RigDispatch_SetMetadata(in Item,in Name,in NameSpace,in Value,out Success)"}
   > RigVMDispatchNode(Element=@n21.Item, Array=RigVMVariableNode() {variable=Backward Controls}.Value) {template="DISPATCH_RigVMDispatch_ArrayAdd(io Array,in Element,out Index)"}
   n22 = RigVMVariableNode() {variable=Control Names As FK}
   Get Item Name = RigVMFunctionReferenceNode(
    Numerical Name Start Index=0
    Remove Side Suffix=true
    Count=@DISPATCH_RigVMDispatch_ArrayIterator.Count
    Name=@DISPATCH_RigVMDispatch_ArrayIterator.Element.Name
    Prefix=RigVMDispatchNode(True=FK, Condition=@n22.Value) {template="DISPATCH_RigVMDispatch_If(in Condition,in True,in False,out Result)"}.Result
    Suffix=RigVMDispatchNode(False=FK, Condition=@n22.Value) {template="DISPATCH_RigVMDispatch_If(in Condition,in True,in False,out Result)"}.Result
    Index=RigVMDispatchNode(Array=RigVMVariableNode() {variable=Controls}.Value) {template="DISPATCH_RigVMDispatch_ArrayGetNum(in Array,out Num)"}.Num
    ExecuteContext=@n15.True
    Standard Numerical Names=@n22.Value
   )
   n23 = RigVMVariableNode() {variable=Shape Settings}
   Get Control Color From Metadata = RigVMFunctionReferenceNode(Bone=@DISPATCH_RigVMDispatch_ArrayIterator.Element, Color Override=@n23.Value.Color)
   Multiply = &RigVMFunction_MathTransformMul(
    B=&RigVMFunction_MathTransformInverse(Value=RigVMVariableNode() {variable=Control Transform Offset}.Value) {template="Inverse::Execute(in Value,out Result)"}.Result
    A=@n23.Value.Transform
   ) {template="Multiply::Execute(in A,in B,out Result)"}
   Scale = &RigVMFunction_MathVectorScale(Value=@Multiply.Result.Scale3D, Factor=@RigVMFunction_AnimEvalRichCurve.Result) {template="Scale::Execute(in Value,in Factor,out Result)"}
   Make = RigVMDispatchNode(
    Elements=(bVisible=True,Name="Circle_Thick",Color=(R=1,G=0,B=0,A=1),Transform=(Rotation=(X=0,Y=-0.7071,Z=0,W=0.7071),Translation=(X=0,Y=0,Z=0),Scale3D=(X=0.4,Y=0.4,Z=0.4)))
    Elements.Name=RigVMDispatchNode(False=Sphere_Solid, Condition=&RigUnit_ShapeExists(ShapeName=@n23.Value.Name).Result, True=@n23.Value.Name) {template="DISPATCH_RigVMDispatch_If(in Condition,in True,in False,out Result)"}.Result
    Elements.Color=@"Get Control Color From Metadata".Result
    Elements.Transform.Rotation=@Multiply.Result.Rotation
    Elements.Transform.Scale3D.Z=@Multiply.Result.Scale3D.Z
    Elements.Transform.Translation=@Multiply.Result.Translation
    Elements.Transform.Scale3D.X=@Scale.Result.X
    Elements.Transform.Scale3D.Y=@Scale.Result.Y
   ) {template="DISPATCH_RigVMDispatch_MakeStruct(in Elements,out Struct)"}
   > SpawnControl = &RigUnit_HierarchyAddControlTransform(
    InitialValue=(Rotation=(X=0,Y=0,Z=0,W=1),Translation=(X=0,Y=0,Z=0),Scale3D=(X=1,Y=1,Z=1))
    OffsetSpace=GlobalSpace
    Settings="(InitialSpace=LocalSpace,bUsePreferredRotationOrder=true,PreferredRotationOrder=YZX,Limits=(LimitTranslationX=(bMinimum=False,bMaximum=False),LimitTranslationY=(bMinimum=False,bMaximum=False),LimitTranslationZ=(bMinimum=False,bMaximum=False),LimitPitch=(bM ...(truncated)"
    Name=@"Get Item Name".Result
    OffsetTransform=@n16.Global
    Settings.Shape=@Make.Struct
    Parent=RigVMVariableNode() {variable=Fk Parent}.Value
    Settings.PreferredRotationOrder=RigVMVariableNode() {variable=Rotation Order}.Value
   ) {template="SpawnControl::Execute(in InitialValue,in Settings,in OffsetTransform,in OffsetSpace,in Parent,in Name,out Item)"}
   > DISPATCH_RigVMDispatch_ArrayAdd = RigVMDispatchNode(Element=@SpawnControl.Item, Array=RigVMVariableNode() {variable=Controls}.Value) {template="DISPATCH_RigVMDispatch_ArrayAdd(io Array,in Element,out Index)"}
   > RigVMDispatchNode(Element=@SpawnControl.Item, Array=RigVMVariableNode() {variable=Backward Controls}.Value) {template="DISPATCH_RigVMDispatch_ArrayAdd(io Array,in Element,out Index)"}
   Int to Name_1 = &RigVMFunction_MathIntToName(PaddedSize=0, Number=@DISPATCH_RigVMDispatch_ArrayIterator.Index) {template="Int to Name::Execute(in Number,in PaddedSize,out Result)"}
   > n24 = &RigUnit_HierarchyAddControlTransform(
    InitialValue=(Rotation=(X=0,Y=0,Z=0,W=1),Translation=(X=0,Y=0,Z=0),Scale3D=(X=1,Y=1,Z=1))
    OffsetSpace=GlobalSpace
    Settings="(InitialSpace=LocalSpace,bUsePreferredRotationOrder=False,PreferredRotationOrder=YZX,Limits=(LimitTranslationX=(bMinimum=False,bMaximum=False),LimitTranslationY=(bMinimum=False,bMaximum=False),LimitTranslationZ=(bMinimum=False,bMaximum=False),LimitPitch=(b ...(truncated)"
    Name=&RigVMFunction_NameConcat(A="Sec FK ", B=@"Int to Name_1".Result) {template="Concat::Execute(in A,in B,out Result)"}.Result
    Settings.Shape=@n20.Struct
    Parent=@SpawnControl.Item
    OffsetTransform=@n16.Global
   ) {template="SpawnControl::Execute(in InitialValue,in Settings,in OffsetTransform,in OffsetSpace,in Parent,in Name,out Item)"}
   > RigVMVariableNode(Value=@n24.Item) {variable=Fk Parent}
   > Add = RigVMDispatchNode(Element=@n24.Item, Array=RigVMVariableNode() {variable=Sec Controls}.Value) {template="DISPATCH_RigVMDispatch_ArrayAdd(io Array,in Element,out Index)"}
   > Set_Metadata = RigVMDispatchNode(
    Name=Offset
    NameSpace=Self
    Value=&RigUnit_GetTransform(Space=LocalSpace, bInitial=False, Item=@n24.Item).Transform.Rotation
    Item=@n24.Item
   ) {template="DISPATCH_RigDispatch_SetMetadata(in Item,in Name,in NameSpace,in Value,out Success)"}
   Concat = &RigVMFunction_NameConcat(B=" Sec", A=@DISPATCH_RigVMDispatch_ArrayIterator.Element.Name) {template="Concat::Execute(in A,in B,out Result)"}
  ]}
 RigVMFunctionLibrary {links [], nodes []}
 RigVMFunction_Sequence_1_ContainedGraph {nodes [
   n1 = RigVMFunctionEntryNode()
   RigVMFunction_Sequence = &RigVMFunction_Sequence(ExecuteContext=@n1.ExecuteContext)
   n2 = &RigVMFunction_Sequence(ExecuteContext=@RigVMFunction_Sequence.B)
   n3 = &RigVMFunction_Sequence(ExecuteContext=@n2.B)
   RigVMFunctionReturnNode(A=@RigVMFunction_Sequence.A, B=@n2.A, C=@n3.A, D=@n3.B)
  ]}
 RigVMFunction_Sequence_1_ContainedGraph {nodes [
   n1 = RigVMFunctionEntryNode()
   RigVMFunction_Sequence = &RigVMFunction_Sequence(ExecuteContext=@n1.ExecuteContext)
   n2 = &RigVMFunction_Sequence(ExecuteContext=@RigVMFunction_Sequence.B)
   n3 = &RigVMFunction_Sequence(ExecuteContext=@n2.B)
   n4 = &RigVMFunction_Sequence(ExecuteContext=@n3.B)
   n5 = &RigVMFunction_Sequence(ExecuteContext=@n4.B)
   RigVMFunctionReturnNode(A=@RigVMFunction_Sequence.A, B=@n2.A, C=@n3.A, D=@n4.A, E=@n5.A, F=@n5.B)
  ]}
 RigVMModel {nodes [
   RigUnit_BeginExecution = &RigUnit_BeginExecution()
   RigVMFunction_Sequence = RigVMAggregateNode(ExecuteContext=@RigUnit_BeginExecution.ExecutePin)
   For_Each = RigVMDispatchNode(ExecuteContext=@RigVMFunction_Sequence.B, Array=RigVMVariableNode() {variable=Controls}.Value) {template="DISPATCH_RigVMDispatch_ArrayIterator(in Array,out Element,out Index,out Count,out Ratio)"}
   > n1 = &RigVMFunction_ForLoopCount(Count=RigVMVariableNode() {variable=Distribution Every}.Value)
   n2 = &RigVMFunction_ControlFlowBranch(
    Condition=RigVMDispatchNode(B=0, A=@n1.Index) {template="DISPATCH_RigVMDispatch_CoreEquals(in A,in B,out Result)"}.Result
    ExecuteContext=@n1.ExecutePin
   )
   n3 = RigVMDispatchNode(
    Array=/Script/ControlRig.RigUnit_CollectionChildrenArray(TypeToSearch=All, bDefaultChildren=True, bIncludeParent=false, bRecursive=true, Parent=@For_Each.Element) {is=struct RigVMUnitNode}.Items
    Index=@n1.Index
   ) {template="DISPATCH_RigVMDispatch_ArrayGetAtIndex(in Array,in Index,out Element)"}
   Get_Metadata = RigVMDispatchNode(Default=(X=0,Y=0,Z=0,W=1), Name=Offset, NameSpace=Self, Item=@n3.Element) {template="DISPATCH_RigDispatch_GetMetadata(in Item,in Name,in NameSpace,in Default,out Value,out Found)"}
   Multiply = /Script/RigVM.RigVMFunction_MathQuaternionMul(
    A=&RigUnit_GetTransform(Space=LocalSpace, bInitial=False, Item=@For_Each.Element).Transform.Rotation
    B=@Get_Metadata.Value
   ) {is=struct RigVMUnitNode, template="Multiply::Execute(in A,in B,out Result)"}
   SetControlOffset = /Script/ControlRig.RigUnit_SetControlRotationOffset(Space=LocalSpace, Control=@n3.Element.Name, Offset=@Multiply.Result, ExecutePin=@n2.False) {is=struct RigVMUnitNode, template="SetControlOffset::Execute(in Control,io Offset,in Space)"}
   n4 = RigVMDispatchNode(ExecuteContext=@RigVMFunction_Sequence.C, Array=RigVMVariableNode() {variable=Bones}.Value) {template="DISPATCH_RigVMDispatch_ArrayIterator(in Array,out Element,out Index,out Count,out Ratio)"}
   Int to Name_1 = &RigVMFunction_MathIntToName(PaddedSize=0, Number=@n4.Index) {template="Int to Name::Execute(in Number,in PaddedSize,out Result)"}
   n5 = &RigVMFunction_NameConcat(A="Sec FK ", B=@"Int to Name_1".Result) {template="Concat::Execute(in A,in B,out Result)"}
   n6 = RigVMVariableNode() {variable=Control Transform Offset}
   > &RigUnit_SetTransform(
    Weight=1
    bInitial=False
    bPropagateToChildren=true
    Item=@n4.Element
    Value=&RigVMFunction_MathTransformMakeAbsolute(
     Parent=&RigUnit_GetTransform(Item=(Type=Control,Name="None"), bInitial=False, Item.Name=@n5.Result).Transform
     Local=&RigVMFunction_MathTransformInverse(Value=@n6.Value) {template="Inverse::Execute(in Value,out Result)"}.Result
    ) {template="Make Absolute::Execute(in Local,in Parent,out Global)"}.Global
   )
   Evaluate Space Nulls = RigVMFunctionReferenceNode(
    Orient Only=true
    Space Nulls=RigVMVariableNode() {variable=Orient Space Nulls}.Value
    Spaces=RigVMVariableNode() {variable=Orient Spaces}.Value
    ExecuteContext=@RigVMFunction_Sequence.A
   )
   > For_Each_1 = RigVMDispatchNode(Array=RigVMVariableNode() {variable=Sec Controls}.Value) {template="DISPATCH_RigVMDispatch_ArrayIterator(in Array,out Element,out Index,out Count,out Ratio)"}
   > SetControlVisibility = /Script/ControlRig.RigUnit_SetControlVisibility(
    Item=@For_Each_1.Element
    bVisible=/Script/ControlRig.RigUnit_GetBoolAnimationChannelFromItem(bInitial=False, Item=RigVMVariableNode() {variable=Sec Control Vis Channel}.Value) {is=struct RigVMUnitNode, template="GetAnimationChannelFromItem::Execute(out Value,in Item,in bInitial)"}.Value
   ) {is=struct RigVMUnitNode, template="SetControlVisibility::Execute(in Item,in Pattern,in bVisible)"}
   GetAnimationChannelFromItem = /Script/ControlRig.RigUnit_GetFloatAnimationChannelFromItem(bInitial=False, Item=RigVMVariableNode() {variable=Dyn Strength Channel}.Value) {is=struct RigVMUnitNode, template="GetAnimationChannelFromItem::Execute(out Value,in Item,in bInitial)"}
   Local Chain Dynamics_1 = RigVMFunctionReferenceNode(
    Blend Twist=0
    Draw Debug=false
    Force=(X=0,Y=0,Z=0)
    Strength=@GetAnimationChannelFromItem.Value
    Damping=&RigUnit_GetFloatAnimationChannelFromItem(bInitial=False, Item=RigVMVariableNode() {variable=Dyn Damping Channel}.Value) {template="GetAnimationChannelFromItem::Execute(out Value,in Item,in bInitial)"}.Value
    Blend Start=&RigUnit_GetFloatAnimationChannelFromItem(bInitial=False, Item=RigVMVariableNode() {variable=Dyn Blend Start Channel}.Value) {template="GetAnimationChannelFromItem::Execute(out Value,in Item,in bInitial)"}.Value
    Blend End=&RigUnit_GetFloatAnimationChannelFromItem(bInitial=False, Item=RigVMVariableNode() {variable=Dyn Blend End Channel}.Value) {template="GetAnimationChannelFromItem::Execute(out Value,in Item,in bInitial)"}.Value
    Primary Axis=RigVMVariableNode() {variable=Primary Axis}.Value
    Bones=RigVMVariableNode() {variable=Bones}.Value
    Secondary Axis=RigVMVariableNode() {variable=Secondary Axis}.Value
    ExecuteContext=@RigVMFunction_Sequence.D
   )
  ]}
 RigVMModel Rig {nodes [
   n1 = /Script/ControlRig.RigUnit_InverseExecution() {is=struct RigVMUnitNode}
   RigVMFunction_Sequence = &RigVMFunction_Sequence(ExecuteContext=@n1.ExecutePin)
   DISPATCH_RigVMDispatch_ArrayIterator = RigVMDispatchNode(ExecuteContext=@RigVMFunction_Sequence.B, Array=RigVMVariableNode() {variable=Bones}.Value) {template="DISPATCH_RigVMDispatch_ArrayIterator(in Array,out Element,out Index,out Count,out Ratio)"}
   DISPATCH_RigVMDispatch_ArrayGetAtIndex = RigVMDispatchNode(Index=@DISPATCH_RigVMDispatch_ArrayIterator.Index, Array=RigVMVariableNode() {variable=Backward Controls}.Value) {template="DISPATCH_RigVMDispatch_ArrayGetAtIndex(in Array,in Index,out Element)"}
   > &RigUnit_SetTransform(
    Weight=1
    bInitial=False
    Item=@DISPATCH_RigVMDispatch_ArrayGetAtIndex.Element
    Value=&RigVMFunction_MathTransformMakeAbsolute(
     Parent=&RigUnit_GetTransform(bInitial=False, Item=@DISPATCH_RigVMDispatch_ArrayIterator.Element).Transform
     Local=RigVMVariableNode() {variable=Control Transform Offset}.Value
    ) {template="Make Absolute::Execute(in Local,in Parent,out Global)"}.Global
   )
   Evaluate Space Nulls = RigVMFunctionReferenceNode(
    Orient Only=true
    Space Nulls=RigVMVariableNode() {variable=Orient Space Nulls}.Value
    Spaces=RigVMVariableNode() {variable=Orient Spaces}.Value
    ExecuteContext=@RigVMFunction_Sequence.A
   )
  ]}
]
hierarchy [
 Start {type=Connector, loc=<0 0 0>, rot=<0 0 0 1>}
 Root {type=Socket, loc=<0 0 0>, rot=<0 0 0 1>}
 End {type=Connector, loc=<0 0 0>, rot=<0 0 0 1>}
 Parent {type=Connector, loc=<0 0 0>, rot=<0 0 0 1>}
 Orient Spaces {type=Connector, loc=<0 0 0>, rot=<0 0 0 1>}
]
variables [
 Backward Controls {type="TArray<FRigElementKey>", default=(())}
 Bones {type="TArray<FRigElementKey>", default=(())}
 Bones To Distribute {type="TArray<FRigElementKey>", default=(())}
 Control Names As FK {type=bool, default=True}
 Control Scale {type=double, default=1}
 Control Scale Factor Profile {
  type=FRuntimeFloatCurve
  default=(EditorCurveData=(Keys=((Value=1),(Time=1,Value=1)),DefaultValue=340282346638528859811704183484516925440,PreInfinityExtrap=RCCE_Constant,PostInfinityExtrap=RCCE_Constant),ExternalCurve=None)
 }
 Control Transform Offset {type=FTransform, default=(Rotation=(X=0,Y=0,Z=0,W=1),Translation=(X=0,Y=0,Z=0),Scale3D=(X=1,Y=1,Z=1))}
 Controls {type="TArray<FRigElementKey>", default=(())}
 Default Orient Space Index {type=int32, default=-1}
 Distribution Every {type=int32, default=2}
 Dyn Blend End Channel {type=FRigElementKey, default=(Type=Bone,Name="")}
 Dyn Blend Start Channel {type=FRigElementKey, default=(Type=Bone,Name="")}
 Dyn Damping Channel {type=FRigElementKey, default=(Type=Bone,Name="")}
 Dyn Strength Channel {type=FRigElementKey, default=(Type=Bone,Name="")}
 Fk Parent {type=FRigElementKey, default=(Type=None,Name="")}
 Main Bones {type="TArray<FRigElementKey>", default=(())}
 Nulls {type="TArray<FRigElementKey>"}
 Orient Space Nulls {type="TArray<FRigElementKey>"}
 Orient Spaces {type="TArray<FRigElementKey>", default=(())}
 Parent {type=FRigElementKey, default=(Type=None,Name="")}
 Previous Int {type=int32, default=0}
 Primary Axis {type=FVector, default=(X=1,Y=0,Z=0)}
 Rotation Order {type=EEulerRotationOrder, default=XYZ}
 Sec Control Vis Channel {type=FRigElementKey, default=(Type=Bone,Name="")}
 Sec Controls {type="TArray<FRigElementKey>", default=(())}
 Sec Shape Settings {
  type=FRigUnit_HierarchyAddControl_ShapeSettings
  default=(bVisible=True,Name="Hexagon_Thick",Color=(R=1,G=0.0737,B=0.0818,A=1),Transform=(Rotation=(X=0,Y=-0.7071,Z=0,W=0.7071),Translation=(X=0,Y=0,Z=0),Scale3D=(X=1.8,Y=1.8,Z=1.8)))
 }
 Secondary Axis {type=FVector, default=(X=0,Y=1,Z=0)}
 Shape Settings {
  type=FRigUnit_HierarchyAddControl_ShapeSettings
  default=(bVisible=True,Name="Circle_Thick",Color=(R=1,G=1,B=1,A=1),Transform=(Rotation=(X=0,Y=-0.7071,Z=0,W=0.7071),Translation=(X=0,Y=0,Z=0),Scale3D=(X=2,Y=2,Z=2)))
 }
]
