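# Control Rig module that builds an FK control chain on bones from Start to End connectors, or Start alone.
# Construction spawns nested null-and-control pairs per bone with curve-profiled shape scale, plus an orient space null offered to the first control.
# Forward solve updates that null's rotation, optionally locks scale and drives bones; active-skeleton mode snaps nulls to bones, applying controls' local transforms on top.
# Backward solve fits controls to bones, and connector events set default matches.
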ControlRigBlueprint /EpicControlRig/Modules/FkSolves/Fk/CRM_Fk_Chain_v01
graphs [
 ConstructionGraph {nodes [
   RigUnit_PrepareForExecution = &RigUnit_PrepareForExecution()
   /Script/ControlRig.RigUnit_SetupShapeLibraryFromUserData(LibraryName=CRSL, LogShapeLibraries=False, NameSpace=CRSL, Path=ShapeLibrary, ExecutePin=@RigUnit_PrepareForExecution.ExecutePin) {is=struct RigVMUnitNode}
   > DISPATCH_RigVMDispatch_ArrayReset = RigVMDispatchNode(Array=RigVMVariableNode() {variable=Controls}.Value) {template="DISPATCH_RigVMDispatch_ArrayReset(io Array)"}
   > RigVMFunction_Sequence = RigVMAggregateNode()
   n1 = &RigUnit_GetTransform(
    bInitial=true
    Item=RigVMDispatchNode(Index=0, Array=RigVMVariableNode() {variable=Bones}.Value) {template="DISPATCH_RigVMDispatch_ArrayGetAtIndex(in Array,in Index,out Element)"}.Element
   )
   n2 = /Script/ControlRig.RigUnit_HierarchyAddBone(
    Name=Orient
    Space=GlobalSpace
    Transform=(Rotation=(X=0,Y=0,Z=0,W=1),Translation=(X=0,Y=0,Z=0),Scale3D=(X=1,Y=1,Z=1))
    Transform.Rotation=&RigUnit_GetTransform(bInitial=true, Item=RigVMVariableNode() {variable=Orient Space Connector}.Value).Transform.Rotation
    Transform.Scale3D=@n1.Transform.Scale3D
    Transform.Translation=@n1.Transform.Translation
    Parent=RigVMVariableNode() {variable=Parent Connector}.Value
    ExecutePin=@RigVMFunction_Sequence.B
   ) {is=struct RigVMUnitNode}
   > RigVMVariableNode(Value=@n2.Item) {variable=Orient Space Null}
   n3 = /Script/ControlRig.RigUnit_AddAvailableSpaces(
    Spaces=((Key=(Type=Control,Name="None"),Label="None"))
    Control=RigVMDispatchNode(Index=0, Array=RigVMVariableNode() {variable=Controls}.Value) {template="DISPATCH_RigVMDispatch_ArrayGetAtIndex(in Array,in Index,out Element)"}.Element
    Spaces.0.Key=RigVMVariableNode() {variable=Orient Space Null}.Value
    ExecutePin=@RigVMFunction_Sequence.D
   ) {is=struct RigVMUnitNode}
   RigVMFunction_ControlFlowBranch = &RigVMFunction_ControlFlowBranch(ExecuteContext=@n3.ExecutePin, Condition=RigVMVariableNode() {variable=Default To Orient Space}.Value)
   /Script/ControlRig.RigUnit_SwitchParent(
    Mode=ParentItem
    bMaintainGlobal=True
    Child=RigVMDispatchNode(Index=0, Array=RigVMVariableNode() {variable=Controls}.Value) {template="DISPATCH_RigVMDispatch_ArrayGetAtIndex(in Array,in Index,out Element)"}.Element
    ExecutePin=@RigVMFunction_ControlFlowBranch.True
    Parent=RigVMVariableNode() {variable=Orient Space Null}.Value
   ) {is=struct RigVMUnitNode}
   n4 = &RigVMFunction_ControlFlowBranch(
    Condition=/Script/ControlRig.RigUnit_ResolveConnector(Connector=(Type=Connector,Name="End"), SkipSocket=False) {is=struct RigVMUnitNode}.bIsConnected
    ExecuteContext=@RigVMFunction_Sequence.A
   )
   RigVMVariableNode(Value=(Type=Connector,Name="Parent"), ExecuteContext=@n4.Completed) {variable=Parent Connector}
   > RigVMVariableNode(Value=(Type=Connector,Name="Parent")) {variable=Fk Parent}
   > RigVMVariableNode(Value=(Type=Connector,Name="OrientSpace")) {variable=Orient Space Connector}
   n5 = /Script/ControlRig.RigUnit_HierarchyGetChainItemArray(End=(Type=Connector,Name="End"), Start=(Type=Connector,Name="Start"), bIncludeEnd=True, bIncludeStart=True, bReverse=False) {is=struct RigVMUnitNode}
   RigVMVariableNode(Value=@n5.Chain, ExecuteContext=@n4.True) {variable=Bones}
   RigVMVariableNode(
    Value=((Type=None,Name="None"))
    Value.0=&RigUnit_ResolveConnector(Connector=(Type=Connector,Name="Start"), SkipSocket=False).Result
    ExecuteContext=@n4.False
   ) {variable=Bones}
   Get_Module_Metadata_1 = RigVMDispatchNode(Default=1, Name=Global Control Scale, NameSpace=Root) {template="DISPATCH_RigDispatch_GetModuleMetadata(in Name,in NameSpace,in Default,out Value,out Found)"}
   Set Control Scale = RigVMFunctionReferenceNode(
    Auto Scale Factor=1
    Global Scale=@Get_Module_Metadata_1.Value
    Scale=RigVMVariableNode() {variable=Control Scale}.Value
    Controls=RigVMVariableNode() {variable=Controls}.Value
    ExecuteContext=@RigVMFunction_Sequence.E
   )
   DISPATCH_RigVMDispatch_ArrayIterator = RigVMDispatchNode(Array=RigVMVariableNode() {variable=Bones}.Value, ExecuteContext=@RigVMFunction_Sequence.C) {template="DISPATCH_RigVMDispatch_ArrayIterator(in Array,out Element,out Index,out Count,out Ratio)"}
   n6 = RigVMVariableNode() {variable=Control Names As FK}
   > Get Item Name v01 = RigVMFunctionReferenceNode(
    Numerical Name Start Index=0
    Remove Side Suffix=true
    Count=@DISPATCH_RigVMDispatch_ArrayIterator.Count
    Name=@DISPATCH_RigVMDispatch_ArrayIterator.Element.Name
    Index=@DISPATCH_RigVMDispatch_ArrayIterator.Index
    Prefix=RigVMDispatchNode(True=FK, Condition=@n6.Value) {template="DISPATCH_RigVMDispatch_If(in Condition,in True,in False,out Result)"}.Result
    Suffix=RigVMDispatchNode(False=FK, Condition=@n6.Value) {template="DISPATCH_RigVMDispatch_If(in Condition,in True,in False,out Result)"}.Result
    Standard Numerical Names=@n6.Value
   )
   > n7 = &RigUnit_HierarchyAddNull(
    Space=GlobalSpace
    Name=@"Get Item Name v01".Result
    Transform=/Script/RigVM.RigVMFunction_MathTransformMakeAbsolute(
     Parent=&RigUnit_GetTransform(bInitial=true, Item=@DISPATCH_RigVMDispatch_ArrayIterator.Element).Transform
     Local=RigVMVariableNode() {variable=Control Transform Offset}.Value
    ) {is=struct RigVMUnitNode, template="Make Absolute::Execute(in Local,in Parent,out Global)"}.Global
    Parent=RigVMVariableNode() {variable=Fk Parent}.Value
   )
   > RigVMDispatchNode(Element=@n7.Item, Array=RigVMVariableNode() {variable=Nulls}.Value) {template="DISPATCH_RigVMDispatch_ArrayAdd(io Array,in Element,out Index)"}
   Get Control Shape = RigVMFunctionReferenceNode(Shape=RigVMVariableNode() {variable=Shape Settings}.Value, Transform Offset=RigVMVariableNode() {variable=Control Transform Offset}.Value)
   Get Control Color From Metadata = RigVMFunctionReferenceNode(Color Override=@"Get Control Shape".Result.Color)
   n8 = RigVMVariableNode() {variable=Display Names}
   From_String = RigVMDispatchNode(String=RigVMDispatchNode(Array=@n8.Value, Index=@DISPATCH_RigVMDispatch_ArrayIterator.Index) {template="DISPATCH_RigVMDispatch_ArrayGetAtIndex(in Array,in Index,out Element)"}.Element) {template="DISPATCH_RigDispatch_FromString(in String,out Result)"}
   n9 = RigVMDispatchNode(Array=@n8.Value) {template="DISPATCH_RigVMDispatch_ArrayGetNum(in Array,out Num)"}
   GreaterEqual = /Script/RigVM.RigVMFunction_MathIntGreaterEqual(B=0, A=@n9.Num) {is=struct RigVMUnitNode, template="GreaterEqual::Execute(in A,in B,out Result)"}
   Less = /Script/RigVM.RigVMFunction_MathIntLess(B=@n9.Num, A=@DISPATCH_RigVMDispatch_ArrayIterator.Index) {is=struct RigVMUnitNode, template="Less::Execute(in A,in B,out Result)"}
   RigVMFunction_MathBoolAnd = &RigVMFunction_MathBoolAnd(B=@GreaterEqual.Result, A=@Less.Result)
   RigVMFunction_AnimEvalRichCurve = &RigVMFunction_AnimEvalRichCurve(
    SourceMinimum=0
    TargetMinimum=0
    Curve=RigVMVariableNode() {variable=Control Scale Factor Profile}.Value
    Value=@DISPATCH_RigVMDispatch_ArrayIterator.Ratio
   )
   Scale = /Script/RigVM.RigVMFunction_MathVectorScale(Value=@"Get Control Shape".Result.Transform.Scale3D, Factor=@RigVMFunction_AnimEvalRichCurve.Result) {is=struct RigVMUnitNode, template="Scale::Execute(in Value,in Factor,out Result)"}
   > SpawnControl = /Script/ControlRig.RigUnit_HierarchyAddControlTransform(
    InitialValue=(Rotation=(X=0,Y=0,Z=0,W=1),Translation=(X=0,Y=0,Z=0),Scale3D=(X=1,Y=1,Z=1))
    OffsetSpace=LocalSpace
    OffsetTransform=(Rotation=(X=0,Y=0,Z=0,W=1),Translation=(X=0,Y=0,Z=0),Scale3D=(X=1,Y=1,Z=1))
    Settings="(InitialSpace=LocalSpace,bUsePreferredRotationOrder=true,PreferredRotationOrder=YZX,Limits=(LimitTranslationX=(bMinimum=False,bMaximum=False),LimitTranslationY=(bMinimum=False,bMaximum=False),LimitTranslationZ=(bMinimum=False,bMaximum=False),LimitPitch=(bM ...(truncated)"
    Settings.Shape.Color=@"Get Control Color From Metadata".Result
    Settings.Shape.Name=@"Get Control Shape".Result.Name
    Settings.Shape.Transform.Rotation=@"Get Control Shape".Result.Transform.Rotation
    Settings.Shape.Transform.Translation=@"Get Control Shape".Result.Transform.Translation
    Settings.Shape.bVisible=@"Get Control Shape".Result.bVisible
    Name=@"Get Item Name v01".Result
    Parent=@n7.Item
    Settings.DisplayName=RigVMDispatchNode(False=None, True=@From_String.Result, Condition=@RigVMFunction_MathBoolAnd.Result) {template="DISPATCH_RigVMDispatch_If(in Condition,in True,in False,out Result)"}.Result
    Settings.Shape.Transform.Scale3D=@Scale.Result
    Settings.PreferredRotationOrder=RigVMVariableNode() {variable=Rotation Order}.Value
   ) {
    is=struct RigVMUnitNode
    template="SpawnControl::Execute(in InitialValue,in Settings,in OffsetTransform,in OffsetSpace,in Parent,in Name,out Item)"
   }
   > RigVMVariableNode(Value=@SpawnControl.Item) {variable=Fk Parent}
   > DISPATCH_RigVMDispatch_ArrayAdd = RigVMDispatchNode(Array=RigVMVariableNode() {variable=Controls}.Value, Element=@SpawnControl.Item) {template="DISPATCH_RigVMDispatch_ArrayAdd(io Array,in Element,out Index)"}
   > Set Mirror Behavior = RigVMFunctionReferenceNode(Value=RigVMVariableNode() {variable=Mirror Behavior}.Value, Item=@SpawnControl.Item)
   > Set Mirror Axis = RigVMFunctionReferenceNode(Value=RigVMVariableNode() {variable=Mirror Axis}.Value, Item=@SpawnControl.Item)
  ]}
 New Function_ContainedGraph {nodes [
   n1 = RigVMFunctionEntryNode()
   Multiply = &RigVMFunction_MathTransformMul(
    A=@n1.Shape.Transform
    B=/Script/RigVM.RigVMFunction_MathTransformInverse(Value=@n1.Transform Offset) {is=struct RigVMUnitNode, template="Inverse::Execute(in Value,out Result)"}.Result
   ) {template="Multiply::Execute(in A,in B,out Result)"}
   > RigVMFunctionReturnNode(
    Result=(bVisible=True,Name="Default",Color=(R=1,G=0,B=0,A=1),Transform=(Rotation=(X=0,Y=0,Z=0,W=1),Translation=(X=0,Y=0,Z=0),Scale3D=(X=1,Y=1,Z=1)))
    Result.Name=RigVMDispatchNode(
     False=Sphere_Solid
     True=@n1.Shape.Name
     Condition=/Script/ControlRig.RigUnit_ShapeExists(ShapeName=@n1.Shape.Name) {is=struct RigVMUnitNode}.Result
    ) {template="DISPATCH_RigVMDispatch_If(in Condition,in True,in False,out Result)"}.Result
    Result.Color=@n1.Shape.Color
    Result.bVisible=@n1.Shape.bVisible
    Result.Transform=@Multiply.Result
   )
  ]}
 New Function_ContainedGraph {nodes [
   n1 = RigVMFunctionEntryNode()
   > Conform Name v01_1 = RigVMFunctionReferenceNode(Name=@n1.Name)
   Int to Name = /Script/RigVM.RigVMFunction_MathIntToName(PaddedSize=0, Number=@n1.Index) {is=struct RigVMUnitNode, template="Int to Name::Execute(in Number,in PaddedSize,out Result)"}
   Concat = /Script/RigVM.RigVMFunction_NameConcat(A="FK ", B=@"Int to Name".Result) {is=struct RigVMUnitNode, template="Concat::Execute(in A,in B,out Result)"}
   Greater = /Script/RigVM.RigVMFunction_MathIntGreater(B=1, A=@n1.Count) {is=struct RigVMUnitNode, template="Greater::Execute(in A,in B,out Result)"}
   DISPATCH_RigVMDispatch_If = RigVMDispatchNode(False=FK, True=@Concat.Result, Condition=@Greater.Result) {template="DISPATCH_RigVMDispatch_If(in Condition,in True,in False,out Result)"}
   > RigVMFunctionReturnNode(Result=RigVMDispatchNode(False=@"Conform Name v01_1".Conformed Name, True=@DISPATCH_RigVMDispatch_If.Result, Condition=@n1.FK Standard Names) {template="DISPATCH_RigVMDispatch_If(in Condition,in True,in False,out Result)"}.Result)
  ]}
 RigVMFunctionLibrary {nodes [
   Get Control Name = RigVMCollapseNode(Count=0, ExecuteContext=(), FK Standard Names=False, Index=0)
   Get Control Shape = RigVMCollapseNode(
    Shape=(bVisible=True,Name="Default",Color=(R=1,G=0,B=0,A=1),Transform=(Rotation=(X=0,Y=0,Z=0,W=1),Translation=(X=0,Y=0,Z=0),Scale3D=(X=1,Y=1,Z=1)))
    Transform Offset=(Rotation=(X=0,Y=0,Z=0,W=1),Translation=(X=0,Y=0,Z=0),Scale3D=(X=1,Y=1,Z=1))
   )
  ]}
 RigVMFunction_Sequence_1_ContainedGraph {nodes [
   n1 = RigVMFunctionEntryNode()
   RigVMFunction_Sequence = &RigVMFunction_Sequence(ExecuteContext=@n1.ExecuteContext)
   n2 = &RigVMFunction_Sequence(ExecuteContext=@RigVMFunction_Sequence.B)
   RigVMFunctionReturnNode(A=@RigVMFunction_Sequence.A, B=@n2.A, C=@n2.B)
  ]}
 RigVMFunction_Sequence_1_ContainedGraph {nodes [
   n1 = RigVMFunctionEntryNode()
   RigVMFunction_Sequence = &RigVMFunction_Sequence(ExecuteContext=@n1.ExecuteContext)
   n2 = &RigVMFunction_Sequence(ExecuteContext=@RigVMFunction_Sequence.B)
   n3 = &RigVMFunction_Sequence(ExecuteContext=@n2.B)
   n4 = &RigVMFunction_Sequence(ExecuteContext=@n3.B)
   RigVMFunctionReturnNode(A=@RigVMFunction_Sequence.A, B=@n2.A, C=@n3.A, D=@n4.A, E=@n4.B)
  ]}
 RigVMModel {nodes [
   RigUnit_BeginExecution = &RigUnit_BeginExecution()
   RigVMFunction_Sequence = RigVMAggregateNode(ExecuteContext=@RigUnit_BeginExecution.ExecutePin)
   RigVMFunction_ControlFlowBranch = &RigVMFunction_ControlFlowBranch(ExecuteContext=@RigVMFunction_Sequence.C, Condition=RigVMVariableNode() {variable=Drive Bones}.Value)
   DISPATCH_RigVMDispatch_ArrayIterator = RigVMDispatchNode(ExecuteContext=@RigVMFunction_ControlFlowBranch.True, Array=RigVMVariableNode() {variable=Bones}.Value) {template="DISPATCH_RigVMDispatch_ArrayIterator(in Array,out Element,out Index,out Count,out Ratio)"}
   n1 = &RigVMFunction_ControlFlowBranch(ExecuteContext=@DISPATCH_RigVMDispatch_ArrayIterator.ExecuteContext, Condition=RigVMVariableNode() {variable=Use Active Skeleton}.Value)
   n2 = RigVMVariableNode() {variable=Control Transform Offset}
   RigVMFunction_MathBoolNot = &RigVMFunction_MathBoolNot(Value=RigVMVariableNode() {variable=Use Active Skeleton}.Value)
   &RigUnit_SetTransform(
    Weight=1
    bInitial=False
    Item=@DISPATCH_RigVMDispatch_ArrayIterator.Element
    Value=&RigVMFunction_MathTransformMakeAbsolute(
     Local=&RigVMFunction_MathTransformInverse(Value=@n2.Value) {template="Inverse::Execute(in Value,out Result)"}.Result
     Parent=RigVMVariableNode() {variable=Bone Transform}.Value
    ) {template="Make Absolute::Execute(in Local,in Parent,out Global)"}.Global
    ExecutePin=@n1.Completed
    bPropagateToChildren=@RigVMFunction_MathBoolNot.Result
   )
   DISPATCH_RigVMDispatch_ArrayGetAtIndex = RigVMDispatchNode(Index=@DISPATCH_RigVMDispatch_ArrayIterator.Index, Array=RigVMVariableNode() {variable=Controls}.Value) {template="DISPATCH_RigVMDispatch_ArrayGetAtIndex(in Array,in Index,out Element)"}
   RigVMVariableNode(
    Value=&RigUnit_GetTransform(bInitial=False, Item=@DISPATCH_RigVMDispatch_ArrayGetAtIndex.Element).Transform
    ExecuteContext=@n1.False
   ) {variable=Bone Transform}
   n3 = RigVMDispatchNode(Index=@DISPATCH_RigVMDispatch_ArrayIterator.Index, Array=RigVMVariableNode() {variable=Nulls}.Value) {template="DISPATCH_RigVMDispatch_ArrayGetAtIndex(in Array,in Index,out Element)"}
   &RigUnit_SetTransform(
    Weight=1
    bInitial=False
    Item=@n3.Element
    Value=&RigUnit_GetTransform(bInitial=False, Item=@DISPATCH_RigVMDispatch_ArrayIterator.Element).Transform
    ExecutePin=@n1.True
   )
   > RigVMVariableNode(Value=&RigVMFunction_MathTransformMakeAbsolute(
     Parent=&RigUnit_GetTransform(bInitial=False, Item=@n3.Element).Transform
     Local=&RigUnit_GetTransform(Space=LocalSpace, bInitial=False, Item=@DISPATCH_RigVMDispatch_ArrayGetAtIndex.Element).Transform
    ) {template="Make Absolute::Execute(in Local,in Parent,out Global)"}.Global) {variable=Bone Transform}
   n4 = RigVMVariableNode() {variable=Parent Connector}
   n5 = RigVMVariableNode() {variable=Orient Space Null}
   n6 = /Script/ControlRig.RigUnit_ProjectTransformToNewParent(bChildInitial=True, bNewParentInitial=False, bOldParentInitial=True, NewParent=@n4.Value, OldParent=@n4.Value, Child=@n5.Value) {is=struct RigVMUnitNode}
   &RigUnit_SetTransform(
    Value=(Rotation=(X=0,Y=0,Z=0,W=1),Translation=(X=0,Y=0,Z=0),Scale3D=(X=1,Y=1,Z=1))
    Weight=1
    bInitial=False
    Value.Rotation=&RigUnit_GetTransform(bInitial=False, Item=RigVMVariableNode() {variable=Orient Space Connector}.Value).Transform.Rotation
    Value.Scale3D=@n6.Transform.Scale3D
    Value.Translation=@n6.Transform.Translation
    Item=@n5.Value
    ExecutePin=@RigVMFunction_Sequence.B
   )
   n7 = &RigVMFunction_ControlFlowBranch(Condition=RigVMVariableNode() {variable=Lock Scale}.Value, ExecuteContext=@RigVMFunction_Sequence.A)
   For_Each = RigVMDispatchNode(ExecuteContext=@n7.True, Array=RigVMVariableNode() {variable=Controls}.Value) {template="DISPATCH_RigVMDispatch_ArrayIterator(in Array,out Element,out Index,out Count,out Ratio)"}
   > /Script/ControlRig.RigUnit_SetScale(Scale=(X=1,Y=1,Z=1), Space=LocalSpace, Weight=1, bInitial=False, bPropagateToChildren=True, Item=@For_Each.Element) {is=struct RigVMUnitNode}
  ]}
 RigVMModel Backward Solve Graph {nodes [
   n1 = /Script/ControlRig.RigUnit_InverseExecution() {is=struct RigVMUnitNode}
   DISPATCH_RigVMDispatch_ArrayIterator = RigVMDispatchNode(ExecuteContext=@n1.ExecutePin, Array=RigVMVariableNode() {variable=Bones}.Value) {template="DISPATCH_RigVMDispatch_ArrayIterator(in Array,out Element,out Index,out Count,out Ratio)"}
   DISPATCH_RigVMDispatch_ArrayGetAtIndex = RigVMDispatchNode(Index=@DISPATCH_RigVMDispatch_ArrayIterator.Index, Array=RigVMVariableNode() {variable=Controls}.Value) {template="DISPATCH_RigVMDispatch_ArrayGetAtIndex(in Array,in Index,out Element)"}
   > &RigUnit_SetTransform(
    Weight=1
    bInitial=False
    Item=@DISPATCH_RigVMDispatch_ArrayGetAtIndex.Element
    Value=&RigVMFunction_MathTransformMakeAbsolute(
     Parent=&RigUnit_GetTransform(bInitial=False, Item=@DISPATCH_RigVMDispatch_ArrayIterator.Element).Transform
     Local=RigVMVariableNode() {variable=Control Transform Offset}.Value
    ) {template="Make Absolute::Execute(in Local,in Parent,out Global)"}.Global
   )
  ]}
 RigVMModel Connector Graph {nodes [
   n1 = /Script/ControlRig.RigUnit_ConnectorExecution() {is=struct RigVMUnitNode}
   Item = /Script/ControlRig.RigUnit_Item(Item=(Type=Connector,Name="Start")) {is=struct RigVMUnitNode}
   Set Default Match To Connector v01_3 = RigVMFunctionReferenceNode(
    Connector=(Type=Connector,Name="End")
    ExecuteContext=@n1.ExecuteContext
    Default Match=RigVMDispatchNode(
     Index=-1
     Array=/Script/ControlRig.RigUnit_CollectionChildrenArray(TypeToSearch=Bone, bDefaultChildren=True, bIncludeParent=false, bRecursive=true, Parent=@Item.Item) {is=struct RigVMUnitNode}.Items
    ) {template="DISPATCH_RigVMDispatch_ArrayGetAtIndex(in Array,in Index,out Element)"}.Element
   )
   > Set Default Match To Connector v01_3_1 = RigVMFunctionReferenceNode(
    Connector=(Type=Connector,Name="Parent")
    Default Match=/Script/ControlRig.RigUnit_HierarchyGetParent(bDefaultParent=True, Child=@Item.Item) {is=struct RigVMUnitNode}.Parent
   )
   DISPATCH_RigDispatch_GetModuleMetadata = RigVMDispatchNode(Default=(Type=Connector,Name="OrientSpace"), Name=BodyControl, NameSpace=Root) {template="DISPATCH_RigDispatch_GetModuleMetadata(in Name,in NameSpace,in Default,out Value,out Found)"}
   > Set Default Match To Connector v01_3_1_1 = RigVMFunctionReferenceNode(Connector=(Type=Connector,Name="OrientSpace"), Default Match=@DISPATCH_RigDispatch_GetModuleMetadata.Value)
  ]}
]
hierarchy [
 Start {type=Connector, loc=<0 0 0>, rot=<0 0 0 1>}
 Root {type=Socket, loc=<0 0 0>, rot=<0 0 0 1>}
 Parent {type=Connector, loc=<0 0 0>, rot=<0 0 0 1>}
 End {type=Connector, loc=<0 0 0>, rot=<0 0 0 1>}
 OrientSpace {type=Connector, loc=<0 0 0>, rot=<0 0 0 1>}
]
variables [
 Bone Transform {type=FTransform, default=(Rotation=(X=0,Y=0,Z=0,W=1),Translation=(X=0,Y=0,Z=0),Scale3D=(X=1,Y=1,Z=1))}
 Bones {type="TArray<FRigElementKey>", default=(())}
 Control Names As FK {type=bool, default=True}
 Control Scale {type=double, default=1}
 Control Scale Factor Profile {
  type=FRuntimeFloatCurve
  default=(EditorCurveData=(Keys=((Value=1),(Time=1,Value=1)),DefaultValue=340282346638528859811704183484516925440,PreInfinityExtrap=RCCE_Constant,PostInfinityExtrap=RCCE_Constant),ExternalCurve=None)
 }
 Control Transform Offset {type=FTransform, default=(Rotation=(X=0,Y=0,Z=0,W=1),Translation=(X=0,Y=0,Z=0),Scale3D=(X=1,Y=1,Z=1))}
 Controls {type="TArray<FRigElementKey>", default=(())}
 Default To Orient Space {type=bool, default=False}
 Display Names {type="TArray<FString>"}
 Drive Bones {type=bool, default=True}
 Fk Parent {type=FRigElementKey, default=(Type=None,Name="")}
 Lock Scale {type=bool, default=False}
 Mirror Axis {type=FVector, default=(X=0,Y=1,Z=1)}
 Mirror Behavior {type=bool, default=False}
 Nulls {type="TArray<FRigElementKey>"}
 Orient Space Connector {type=FRigElementKey, default=(Type=None,Name="")}
 Orient Space Null {type=FRigElementKey, default=(Type=None,Name="")}
 Parent Connector {type=FRigElementKey, default=(Type=None,Name="")}
 Rotation Order {type=EEulerRotationOrder, default=XYZ}
 Shape Settings {
  type=FRigUnit_HierarchyAddControl_ShapeSettings
  default=(bVisible=True,Name="Circle_Thin",Color=(R=1,G=1,B=1,A=1),Transform=(Rotation=(X=0,Y=-0.7071,Z=0,W=0.7071),Translation=(X=0,Y=0,Z=0),Scale3D=(X=1,Y=1,Z=1)))
 }
 Use Active Skeleton {type=bool, default=False}
]
